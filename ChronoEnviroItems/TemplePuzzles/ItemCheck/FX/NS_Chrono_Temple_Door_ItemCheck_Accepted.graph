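# Niagara effect for the Chrono temple door item check, named as the feedback for an accepted item.
# Burst-spawns two short-lived emitters: a mesh Beam that grows along X and a sprite Glow behind it, both offset toward the camera.
# Both tint from a system-level color chosen from an array, with reduced saturation via hue shift.
# Curves flash brightness early, then dim it and fade alpha out.
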NiagaraSystem /ChronoEnviroItems/TemplePuzzles/ItemCheck/FX/NS_Chrono_Temple_Door_ItemCheck_Accepted
dependencies {
 usedDataInterfaces [&NiagaraDataInterfaceCurve, &NiagaraDataInterfaceVectorCurve, &NiagaraDataInterfaceArrayColor]
 usedDynamicInputs [&SelectLinearColorFromArray]
 usedModules [
  &self:SystemSpawnScript.SystemScriptSource.SystemScriptGraph.NiagaraNodeAssignment_0.SetVariables_#1
  &SystemState
 ]
}
emitters {
 Beam {
  emitterUpdateScript [
   &EmitterState(Life Cycle Mode=::NewEnumerator1, Loop Behavior=::NewEnumerator1, Loop Duration=1, UseLoopDelay=false)
   &SpawnBurst_Instantaneous(Use Spawn Probability=false, UseLoopCountLimit=false)
   /SetVariables(Emitter.Desaturation=0.75, Emitter.SpeedY=-0.25) {isSetParameters=true}
  ]
  particleSpawnScript [
   &InitializeParticle(
    Color=&HueShiftLinearColor(Hue=0, Saturation=0.5, Value=1, Linear Color=$System.ColorArray)
    Lifetime=0.8
    Mesh Scale Mode=::NewEnumerator4
    UsePositionOffset=false
    Write Material Random=true
   )
   &InitialMeshOrientation(Mesh Axis To Orient=<0 0 -1>, Mesh Orientation Mode=::NewEnumerator2, Use Rotation Vector=true)
   &SystemLocation(MaskSpawnGroup=false, Offset=<0 35 88>)
  ]
  particleUpdateScript [
   &ParticleState()
   &ScaleMeshSize(Scale Factor=&VectorFromCurve(VectorCurve=DataInterface() {propertyValues {
       XCurve {keys [
         {interpMode=RCIM_Cubic, leaveTangent=0.6159, tangentMode=RCTM_Break, time=0, value=0}
         {arriveTangent=3.7177, interpMode=RCIM_Cubic, tangentMode=RCTM_Break, time=0.3, value=0.5}
         <at 1 0.5 RCIM_Cubic>
        ]}
       YCurve {keys [<at 0 1>,<at 1 1>]}
       ZCurve {keys [<at 0 1>,<at 1 1>]}
      }}))
   &DynamicMaterialParameters(
    Index 0 Param 1=55
    Module.Param0WriteEnabled=true
    Module.Param1WriteEnabled=true
    Module.Param2WriteEnabled=true
    Module.Param3WriteEnabled=true
   ) {module=DynamicMaterialParameters001}
   &ScaleColor(
    Scale Alpha=&FloatFromCurve(FloatCurve=DataInterface() {propertyValues {Curve {keys [<at 0 9>, {arriveTangent=-122.5, interpMode=RCIM_Cubic, leaveTangent=-47.858, tangentMode=RCTM_Break, time=0.5, value=9}, <at 1 0>]}}})
    Scale RGB=&VectorFromFloat(Value=&FloatFromCurve(FloatCurve=DataInterface() {propertyValues {Curve {keys [
          {arriveTangent=4.7164, interpMode=RCIM_Cubic, leaveTangent=4.7164, tangentMode=RCTM_User, time=0, value=0.2}
          {arriveTangent=26.7621, interpMode=RCIM_Cubic, leaveTangent=-2.7552, tangentMode=RCTM_Break, time=0.1408, value=3}
          {arriveTangent=-0.0405, interpMode=RCIM_Cubic, leaveTangent=-0.0405, tangentMode=RCTM_User, time=1, value=0.2}
         ]}}}))
    ScaleA=true
    ScaleRGB=true
   )
   &CameraOffset(Camera Offset Amount=35)
  ]
  rendererClasses [&NiagaraMeshRendererProperties]
  renderers [{rendererClass=&NiagaraMeshRendererProperties, rendererIndex=0}]
 }
 Glow {
  emitterUpdateScript [
   &EmitterState(Life Cycle Mode=::NewEnumerator1, Loop Behavior=::NewEnumerator1, Loop Duration=1, UseLoopDelay=false)
   &SpawnBurst_Instantaneous(Use Spawn Probability=false, UseLoopCountLimit=false)
  ]
  particleSpawnScript [
   &InitializeParticle(
    Color=&HueShiftLinearColor(Hue=0, Saturation=0.7, Value=1, Linear Color=$System.ColorArray)
    Lifetime=0.8
    Sprite Size=<333 170>
    Sprite Size Mode=::NewEnumerator4
    UsePositionOffset=false
    Write Material Random=true
   )
   &SystemLocation(MaskSpawnGroup=false, Offset=<0 35 105>)
  ]
  particleUpdateScript [
   &ParticleState()
   &SolveForcesAndVelocity(Clamp Velocity=false, Limit Acceleration=false)
   &CameraOffset(Camera Offset Amount=55)
   &ScaleColor(
    Scale Alpha=&FloatFromCurve(FloatCurve=DataInterface() {propertyValues {Curve {keys [{interpMode=RCIM_Cubic, tangentMode=RCTM_User, time=0, value=1}, <at 0.8 1>, {interpMode=RCIM_Cubic, tangentMode=RCTM_User, time=1, value=0}]}}})
    Scale RGB=&VectorFromFloat(Value=&FloatFromCurve(FloatCurve=DataInterface() {propertyValues {Curve {keys [
          {arriveTangent=4.7164, interpMode=RCIM_Cubic, leaveTangent=4.7164, tangentMode=RCTM_User, time=0, value=0.2}
          {arriveTangent=26.7621, interpMode=RCIM_Cubic, leaveTangent=-2.7552, tangentMode=RCTM_Break, time=0.1408, value=1.5}
          {arriveTangent=-0.0405, interpMode=RCIM_Cubic, leaveTangent=-0.0405, tangentMode=RCTM_User, time=1, value=0.2}
         ]}}}))
    ScaleA=true
    ScaleRGB=true
   ) {module=ScaleColor001}
  ]
  rendererClasses [&NiagaraSpriteRendererProperties]
  renderers [{rendererClass=&NiagaraSpriteRendererProperties}]
 }
}
summary {
 systemName=NS_Chrono_Temple_Door_ItemCheck_Accepted
 emitters [
  {emitterName=Glow, rendererClasses [&NiagaraSpriteRendererProperties]}
  {emitterName=Beam, rendererClasses [&NiagaraMeshRendererProperties]}
 ]
 userVariables [
  User.bIsFrontEnd {type=NiagaraBool}
  User.bIsFrontEndPreview {type=NiagaraBool}
  User.ColorFX {type=NiagaraInt32}
  User.Scale {type=NiagaraFloat, defaultValue=1.3, description=""}
 ]
}
